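# Niagara idle effect for the BronzeHat character head, burning fire at two sockets selected by spawn group.
# FireMain spawns socket-locked, growing meshes with two color layers and stores a time-smoothed position in a particle attribute.
# FireParticle adds rotating flipbook sprites that drift outward, upward and slightly along owner velocity.
# Both fade in with system age and feed a front-end preview flag into material parameters.
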NiagaraSystem /BRCosmetics/Characters/Player/Male/Medium/Heads/M_MED_BronzeHat_Head/FX/NS_Character_BronzeHat_Idle
dependencies {
 usedDataInterfaces [
  &NiagaraDataInterfaceArrayFloat3
  &NiagaraDataInterfaceColorCurve
  &NiagaraDataInterfaceCurve
  &NiagaraDataInterfaceVectorCurve
  &NiagaraDataInterfaceArrayFloat
  &NiagaraDataInterfaceSocketReader
 ]
 usedDynamicInputs [&Vector2DFromFloat]
 usedModules [
  &self:SystemSpawnScript.SystemScriptSource.SystemScriptGraph.NiagaraNodeAssignment_1.SetVariables_#1
  &SystemState
 ]
}
emitters {
 FireMain {
  emitterUpdateScript [
   &EmitterState(
    Inactive Response=::NewEnumerator1
    Life Cycle Mode=::NewEnumerator1
    Loop Behavior=::NewEnumerator1
    Loop Duration Mode=::NewEnumerator1
    UseLoopDelay=false
   )
   &SpawnRate(SpawnRate=5, Use Spawn Probability=false)
   &SpawnBurst_Instantaneous(Spawn Count=4, Use Spawn Probability=false, UseLoopCountLimit=false)
  ]
  particleSpawnScript [
   &FN_SetSpawnGroup(Number of Spawn Groups=2)
   &InitializeParticle(
    Color=<rgba 0.02 0.0003 0 0.6>
    Lifetime Max=1.5
    Lifetime Mode=::NewEnumerator1
    Mesh Scale=<0.3 0.3 0.17>
    Mesh Scale Mode=::NewEnumerator4
    UsePositionOffset=true
    Write Material Random=true
   )
   &InitialMeshOrientation(
    Mesh Orientation Mode=::NewEnumerator2
    Rotation=&SelectVectorFromArray(
     Direct Array Index=&SpawnGroup()
     Vector Selection Array=DataInterface() {propertyValues {FloatData [<0 0.045 0>, <1.1802 0.8287 0.1036>]}}
    )
    Use Rotation Vector=true
   )
   /SetVariables(Particles.CameraOffset002=1) {isSetParameters=true}
   /SetVariables(Particles.CameraOffset003=2) {isSetParameters=true}
   &DynamicMaterialParameters(
    Index 0 Param 1=&RandomRangeFloat(Minimum=0, Maximum=1)
    Index 0 Param 2=&RandomRangeFloat(Minimum=0, Maximum=1)
    Index 0 Param 3=0.2
    Index 0 Param 4=&MakeCustomFloatFromBool(Bool=$User.bIsFrontEndPreview, True Float=0, False Float=1)
    Module.Index0Float0Write=true
    Module.Index0Float1Write=true
    Module.Index0Float2Write=true
    Module.Index0Float3Write=true
   )
  ]
  particleUpdateScript [
   &ParticleState()
   /SetVariables(Particles.Color001:LinearColor=&MultiplyLinearColor(
     Linear Color=&ColorFromCurve(ColorCurve=DataInterface() {propertyValues {
        AlphaCurve {keys [<at 0 0 RCIM_Cubic>,<at 0.1396 0.8>,<at 0.7438 0.6>,<at 1 0 RCIM_Cubic>]}
        BlueCurve {keys [<at 0 0.0203 RCIM_Cubic>,<at 1 0.0234 RCIM_Cubic>]}
        GreenCurve {keys [<at 0 0.0583 RCIM_Cubic>,<at 1 0.0756 RCIM_Cubic>]}
        RedCurve {keys [<at 0 0.3 RCIM_Cubic>,<at 1 0.3 RCIM_Cubic>]}
       }})
     Scale Factor=&MakeVector4(
      X=1
      Y=1
      Z=1
      W=&FloatFromCurve(FloatCurve=DataInterface() {propertyValues {Curve {keys [<at 0.35 0>,<at 1.1 1>]}}}, CurveIndex=$Engine.System.Age)
     )
    )) {isSetParameters=true}
   /SetVariables(Particles.Color002=&MultiplyLinearColor(
     Linear Color=&ColorFromCurve(ColorCurve=DataInterface() {propertyValues {
        AlphaCurve {keys [<at 0 0>, <at 0.1364 0.6>, {arriveTangent=-0.6947, interpMode=RCIM_Cubic, leaveTangent=-0.6947, time=0.7422, value=0.3}, <at 1 0 RCIM_Cubic>]}
        BlueCurve {keys [<at 0 0.0644 RCIM_Cubic>,<at 0.5642 0.0114 RCIM_Cubic>]}
        GreenCurve {keys [<at 0 1.5499 RCIM_Cubic>,<at 0.5642 0.6219 RCIM_Cubic>]}
        RedCurve {keys [<at 0 5 RCIM_Cubic>,<at 0.5642 3 RCIM_Cubic>]}
       }})
     Scale Factor=&MakeVector4(
      X=1
      Y=1
      Z=1
      W=&FloatFromCurve(FloatCurve=DataInterface() {propertyValues {Curve {keys [<at 0.35 0>,<at 1.1 1>]}}}, CurveIndex=$Engine.System.Age)
     )
    )) {isSetParameters=true}
   &ScaleMeshSize(Scale Factor=&VectorFromCurve(
     VectorCurve=DataInterface() {propertyValues {
       XCurve {keys [<at 0 0.625 RCIM_Cubic>, {arriveTangent=-0.0638, interpMode=RCIM_Cubic, leaveTangent=-0.0638, tangentMode=RCTM_User, time=1, value=1.1}]}
       YCurve {keys [<at 0 0.625 RCIM_Cubic>, {arriveTangent=-0.0638, interpMode=RCIM_Cubic, leaveTangent=-0.0638, tangentMode=RCTM_User, time=1, value=1.1}]}
       ZCurve {keys [<at 0 0.625 RCIM_Cubic>, {arriveTangent=-0.0638, interpMode=RCIM_Cubic, leaveTangent=-0.0638, tangentMode=RCTM_User, time=1, value=1.1}]}
      }}
     Scale Curve=<0.75 0.75 0.75>
    ))
   &SocketLocation(
    SamplePosOffset=false
    Socket Index=&SpawnGroup()
    Socket Reader=$System.Socket Reader
    Socket Sampling Mode=::NewEnumerator2
    VelocitySpeedLimit=false
   )
   &InterpolateOverTime(Rate Of Change=8, Target Vector Value:NiagaraPosition=$Particles.Position, Variable Type=::NewEnumerator2)
   /SetVariables(Particles.Interpolate Dynamic=&MakeFloat4FromVectorAndFloat("Vector (RGB)"=$Particles.InterpolateOverTime.InterpolateOverTime_V3_.Moving Average, "Float (Alpha)"=0.2)) {isSetParameters=true}
  ]
  rendererClasses [&NiagaraMeshRendererProperties]
  renderers [
   {rendererClass=&NiagaraMeshRendererProperties, rendererIndex=0}
   {rendererClass=&NiagaraMeshRendererProperties, rendererIndex=1}
  ]
 }
 FireParticle {
  emitterUpdateScript [
   &EmitterState(Inactive Response=::NewEnumerator1, Life Cycle Mode=::NewEnumerator1, Loop Duration=1, UseLoopDelay=false)
   &SpawnRate(SpawnRate=8, Use Spawn Probability=false)
  ]
  particleSpawnScript [
   &FN_SetSpawnGroup(Number of Spawn Groups=2)
   &InitializeParticle(
    Color=<rgba 1 0.8504 0 1.5>
    Lifetime Min=1.6
    Lifetime Mode=::NewEnumerator1
    Sprite Rotation Mode=::NewEnumerator2
    Sprite Size Max=<25 25>
    Sprite Size Min=<15 15>
    Sprite Size Mode=::NewEnumerator2
    UsePositionOffset=false
    Write Material Random=true
   )
   &CameraOffset(Camera Offset Amount=2)
   &DynamicMaterialParameters(
    Index 0 Param 4=&MakeCustomFloatFromBool(Bool=$User.bIsFrontEndPreview, True Float=0, False Float=1)
    Module.Index0Float0Write=true
    Module.Index0Float1Write=true
    Module.Index0Float2Write=true
    Module.Index0Float3Write=true
   )
  ]
  particleUpdateScript [
   &ParticleState()
   &SpriteRotationRate(Rotation Rate=&RandomRangeFloat(Minimum=-155, Maximum=155))
   &SocketLocation(
    SamplePosOffset=false
    Socket Index=&SpawnGroup()
    Socket Reader=$System.Socket Reader
    Socket Sampling Mode=::NewEnumerator2
    VelocitySpeedLimit=false
   )
   /SetVariables(Particles.SubImageIndex=&FloatFromCurve(FloatCurve=DataInterface() {propertyValues {Curve {keys [<at 0 5>,<at 1 15>]}}})) {isSetParameters=true}
   &ScaleSpriteSize(
    Uniform Curve Scale=0.204
    Uniform Curve Sprite Scale=DataInterface() {propertyValues {Curve {keys [
        {arriveTangent=3.7323, interpMode=RCIM_Cubic, leaveTangent=3.7323, tangentMode=RCTM_User, time=0, value=0}
        {tangentMode=RCTM_User, time=1, value=1}
       ]}}}
   )
   &ScaleColor(
    Scale Alpha=&FloatFromCurve(
     FloatCurve=DataInterface() {propertyValues {Curve {keys [
         <at 0 0 RCIM_Cubic>
         <at 0.2 1 RCIM_Cubic>
         {arriveTangent=-1.4238, interpMode=RCIM_Cubic, leaveTangent=-1.4238, tangentMode=RCTM_User, time=1, value=0}
        ]}}}
     Scale Curve=0.09
    )
    Scale RGB=&MakeVectorFromLinearColorRGB(Color=&ColorFromCurve(ColorCurve=DataInterface() {propertyValues {
        AlphaCurve {keys [<at 0 1 RCIM_Cubic>]}
        BlueCurve {keys [<at 0.1048 0.002 RCIM_Cubic>,<at 0.5156 0 RCIM_Cubic>]}
        GreenCurve {keys [<at 0.1048 0.5644 RCIM_Cubic>,<at 0.5156 0.1401 RCIM_Cubic>]}
        RedCurve {keys [<at 0.1048 1 RCIM_Cubic>,<at 0.5156 1 RCIM_Cubic>]}
       }}))
    ScaleA=true
    ScaleRGB=true
   )
   /SetVariables(Particles.Position=&AddVectorToPosition(
     Position=$Particles.Position
     Vector=&Lerp_Vector(
      A=<0 0 0>
      B=&Add_Vector(
       A=&Multiply_Vector(
        A=&MakeVector(
         X=&SelectFloatFromArray(
          Direct Array Index=&SpawnGroup()
          Float Selection Array=DataInterface() {propertyValues {FloatData [-0.25,0.25]}}
         )
         Y=&SelectFloatFromArray(Direct Array Index=&SpawnGroup(), Float Selection Array=DataInterface() {propertyValues {FloatData [0,0]}})
         Z=&SelectFloatFromArray(
          Direct Array Index=&SpawnGroup()
          Float Selection Array=DataInterface() {propertyValues {FloatData [0.13,0.13]}}
         )
        )
        B=&RandomRangeVector(Minimum=<18 18 18>, Maximum=<20 20 20>)
       )
       B=&Multiply_Vector(
        A=&NormalizeVector(Vector To Normalize=$Engine.Owner.Velocity, Safe Normalize=true, Safe Fallback Threshold=0, Safe Fallback Vector=<0 0 0.001>)
        B=&VectorFromFloat(Value=&FloatFromCurve(
          FloatCurve=DataInterface() {propertyValues {Curve {keys [{interpMode=RCIM_Cubic, tangentMode=RCTM_User, time=0, value=0}, <at 1 1>]}}}
          Scale Curve=2
         ))
       )
      )
      Alpha=&FloatFromCurve(FloatCurve=DataInterface() {propertyValues {Curve {keys [
           {arriveTangent=0.3072, interpMode=RCIM_Cubic, leaveTangent=0.3072, tangentMode=RCTM_User, time=0, value=0}
           {arriveTangent=1.042, tangentMode=RCTM_User, time=1, value=1}
          ]}}})
     )
    )) {isSetParameters=true}
   &ScaleColor(
    Scale Alpha=&FloatFromCurve(FloatCurve=DataInterface() {propertyValues {Curve {keys [<at 0.35 0>,<at 1.1 1>]}}}, CurveIndex=$Engine.System.Age)
    ScaleA=true
    ScaleRGB=true
   ) {module=ScaleColor001}
  ]
  rendererClasses [&NiagaraSpriteRendererProperties]
  renderers [{rendererClass=&NiagaraSpriteRendererProperties}]
 }
}
summary {
 systemName=NS_Character_BronzeHat_Idle
 emitters [
  {emitterName=FireMain, rendererClasses [&NiagaraMeshRendererProperties]}
  {emitterName=FireParticle, rendererClasses [&NiagaraSpriteRendererProperties]}
 ]
 userVariables [
  User.bIsFrontEnd {type=NiagaraBool}
  User.bIsFrontEndPreview {type=NiagaraBool}
 ]
}
